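; Basketball hoop prop that scores shots entering its trigger box and runs a streak-scoring minigame.
; The server grades balls two or three points by distance, totals consecutive shots via ball serial numbers, and reports three-pointers through PlayerScored.
; A replicated message struct's OnRep prints formatted results to clients' toy message feeds.
; Balls receive net lerp positions for their scoring cue; other tagged throw toys score through the minigame interface.
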(event OnComponentBeginOverlap(HoopTriggerBox)
  (params (OverlappedComponent :type Primitive%20Component%20Object%20Reference) (OtherActor :type Actor%20Object%20Reference) (OtherComp :type Primitive%20Component%20Object%20Reference) (OtherBodyIndex :type Integer) (bFromSweep :type Boolean) (SweepResult :type Hit%20Result%20Structure%20%28by%20ref%29))
  (bind (_startposition _endposition) (CallFunction|GetHoopLerpPositions))
  (bind (_startposition_2 _endposition_2) (CallFunction|GetHoopLerpPositions))
  ;[0.4669 0.1073 0.5885] Basket Triggered
  (Networking|SwitchHasAuthority
    (:Authority
      (bind _asb_toy_basketball_base (Utilities|Casting|CastToB_ToyBasketball_Base OtherActor)
        (:then
          (bind (_numpoints _totaldistance) (CallFunction|GradeScore _asb_toy_basketball_base))
          (bind _asfort_player_state_athena (Utilities|Casting|CastToFortPlayerStateAthena (Actor|GetOwner _asb_toy_basketball_base))
            (:then
              _)
            (:CastFailed
              _))
          (bind _output_get (Variables|Default|SetPC%20That%20Scored (Utilities|Casting|CastToFortPlayerController (Actor|GetOwner (Actor|GetOwner _asb_toy_basketball_base)))))
          (bind _totalscore (CallFunction|AccumulateScore _output_get _numpoints _asb_toy_basketball_base))
          (CallFunction|CreateMessage _asfort_player_state_athena _totaldistance _numpoints _totalscore)
          (Class|BToyBasketballBase|PlayScoringCueandDestruct _asb_toy_basketball_base _numpoints _startposition _endposition)
          ;[1 1 1] Check if 3pt Distance
          (if (> _totaldistance (Variables|Default|GetThreePointThresholdDistanceInM))
            ;[1 1 1] Tell Parent Player Scored 3 pointer
            (CallFunction|PlayerScored (Variables|Default|GetPC%20That%20Scored))))
        (:CastFailed
          ;[1 0.1269 0.916] Throwingtoy Gameplay, play a cue and destroy the actor when a player manages to score. We need to delay before destroying to ensure that it plays
          (Utilities|IsValid OtherActor
            (:Is%20Valid
              (bind _returnvalue (GameplayTags|HasMatchingGameplayTag(Message) OtherActor "(TagName=\"Creative.Event.ThrowToy.Score.Basketball\")"))
              (if _returnvalue
                (GameplayCue|ExecuteGameplayCue OtherActor (Variables|Default|GetScore%20Cue))
                (Networking|ForceNetUpdate OtherActor)
                (bind _asminigame_basketball_interface (Utilities|Casting|CastToMinigameBasketballInterface OtherActor))
                (Class|MinigameBasketballInterface|PlayScoringCueAndDestruct _asminigame_basketball_interface _startposition_2 _endposition_2)))))))))

(fn AccumulateScore
  (params (PC :type Fort%20Player%20Controller%20Object%20Reference) (NumPoints :type Integer) (Ball :type B%20Toy%20Basketball%20Base%20Object%20Reference))
  (bind (_issequential _runningscore) (CallFunction|IsSequentialShot PC Ball))
  (if _issequential
    (LocalVariables|SetRunningTotal _runningscore)
    (else
      _))
  (LocalVariables|SetRunningTotal (+ (LocalVariables|GetRunningTotal) NumPoints))
  (Utilities|Map|Add (Variables|Default|GetScoreMap) PC)
  (return (LocalVariables|GetRunningTotal)))

(fn CreateMessage
  (params (PlayerWhoShot :type Fort%20Player%20State%20Object%20Reference) (TotalDistance :type Float%20%28double-precision%29) (ShotScore :type Integer) (TotalScore :type Integer))
  (if (Variables|/Script/FortniteGame.FortRuntimeOptions|GetbToyMessagingEnabled (Options|GetRuntimeOptions))
    (Variables|Default|SetDisplayMessage (Utilities|Struct|MakeStructBasketballMinigameMessage PlayerWhoShot TotalDistance ShotScore TotalScore (Variables|Default|GetMessageSN)))
    (Variables|Default|SetMessageSN (Math|Byte|Byte+Byte (Variables|Default|GetMessageSN) 1))))

(fn GetHoopLerpPositions
  (bind _staticmeshcomponent (Variables|Debug|GetStaticMeshComponent))
  (bind _hooptriggerbox (Variables|Default|GetHoopTriggerBox))
  ;[1 0.7981 0.2486] Returns the start and end position of where and actor should lerp through the net
  (Utilities|IsValid _staticmeshcomponent
    (:Is%20Valid
      (Utilities|IsValid _hooptriggerbox
        (:Is%20Valid
          (return (Transformation|GetWorldLocation _hooptriggerbox) (+ (Transformation|GetWorldLocation _hooptriggerbox) (* (Transformation|GetActorUpVector) (* (Variables|Default|GetHoopLerpDistance) (.z (Transformation|GetWorldScale _staticmeshcomponent)))))))))))

(fn GradeScore
  (params (Ball :type B%20Toy%20Basketball%20Base%20Object%20Reference))
  (LocalVariables|SetMetersTraveled (/ (Math|Vector|VectorLengthXY (- (Transformation|GetWorldLocation (Variables|Default|GetHoopTriggerBox)) (Variables|/Game/Building/ActorBlueprints/Prop/PhysicsBall_Master.PhysicsBall_Master_C|GetInitialSpawnPosition Ball))) 1e+02))
  (return (select (> (LocalVariables|GetMetersTraveled) (Variables|Default|GetThreePointThresholdDistanceInM)) 3 2) (LocalVariables|GetMetersTraveled)))

(fn IsSequentialShot
  (params (PC :type Fort%20Player%20Controller%20Object%20Reference) (Ball :type B%20Toy%20Basketball%20Base%20Object%20Reference))
  (bind _scoremap (Variables|Default|GetScoreMap))
  (bind (_value _returnvalue) (Utilities|Map|Find _scoremap PC))
  (bind (_lastserialnumber_2_ef5d780b4020e911ab429d9ba6b977ba _lastshotlocation_5_a600d2a048cb01620b914bb4107a6cd5 _totalscore_8_9facd4014a6f7fdaaa915396bd79b184) (Utilities|Struct|BreakStructBasketballMinigameState _value))
  (if _returnvalue
    (return (== _lastserialnumber_2_ef5d780b4020e911ab429d9ba6b977ba (- (Variables|/Game/Abilities/Toys/Basketball/Shared/B_ToyBasketball_Base.B_ToyBasketball_Base_C|GetBallSerialNumber Ball) 1)) _totalscore_8_9facd4014a6f7fdaaa915396bd79b184)
    (else
      (return false 0))))

(fn OnRep_DisplayMessage
  (bind _returnvalue (BlueprintContext|GetFortToyContext))
  (bind (_displaymessage_playerwhoshot_21_ef5d780b4020e911ab429d9ba6b977ba _displaymessage_totaldistance_13_a600d2a048cb01620b914bb4107a6cd5 _displaymessage_shotscore_16_9facd4014a6f7fdaaa915396bd79b184 _displaymessage_totalscore_15_54415216430c45042af683a06f9801e1) (Utilities|Struct|BreakStructBasketballMinigameMessage (Variables|GetDisplayMessage_PlayerWhoShot_21_EF5D780B4020E911AB429D9BA6B977BA)))
  (Networking|SwitchHasAuthority
    (:Remote
      (Utilities|IsValid _displaymessage_playerwhoshot_21_ef5d780b4020e911ab429d9ba6b977ba
        (:Is%20Valid
          (bind _returnvalue_2 (Class|FortToyContext|FormatPlayerNameforDisplay (BlueprintContext|GetFortToyContext) _displaymessage_playerwhoshot_21_ef5d780b4020e911ab429d9ba6b977ba "ToyOwner"))
          (bind _formatstring (CallFunction|PickFormatString))
          (Utilities|IsValid _returnvalue
            (:Is%20Valid
              (Class|FortToyContext|ConditionalPrinttoMessageFeed _returnvalue _displaymessage_playerwhoshot_21_ef5d780b4020e911ab429d9ba6b977ba (Transformation|GetActorLocation) (select (> _displaymessage_totaldistance_13_a600d2a048cb01620b914bb4107a6cd5 (Variables|Default|GetMinDistanceForTeamRelevance)) "SquadAndTeamMembers" "SquadMembersOnly") (Utilities|Text|FormatText _formatstring _returnvalue_2 (Math|Float|Round _displaymessage_totaldistance_13_a600d2a048cb01620b914bb4107a6cd5) _displaymessage_shotscore_16_9facd4014a6f7fdaaa915396bd79b184 _displaymessage_totalscore_15_54415216430c45042af683a06f9801e1)))))))))

(fn PickFormatString
  (bind (_displaymessage_playerwhoshot_21_ef5d780b4020e911ab429d9ba6b977ba _displaymessage_totaldistance_13_a600d2a048cb01620b914bb4107a6cd5 _displaymessage_shotscore_16_9facd4014a6f7fdaaa915396bd79b184 _displaymessage_totalscore_15_54415216430c45042af683a06f9801e1) (Utilities|Struct|BreakStructBasketballMinigameMessage (Variables|GetDisplayMessage_PlayerWhoShot_21_EF5D780B4020E911AB429D9BA6B977BA)))
  (bind _returnvalue (!= _displaymessage_shotscore_16_9facd4014a6f7fdaaa915396bd79b184 _displaymessage_totalscore_15_54415216430c45042af683a06f9801e1))
  (bind _returnvalue_2 (> _displaymessage_totalscore_15_54415216430c45042af683a06f9801e1 0))
  (bind _returnvalue_3 (and _returnvalue _returnvalue_2))
  (if (> _displaymessage_totaldistance_13_a600d2a048cb01620b914bb4107a6cd5 (Variables|Default|GetMinDistanceToPrint))
    (return (select _returnvalue_3 (Variables|Default|GetSubsequentShot_WithDistance) (Variables|Default|GetFirstShot_WithDistance)))
    (else
      (return (select _returnvalue_3 (Variables|Default|GetSubsequentShot_NoDistance) (Variables|Default|GetFirstShot_NoDistance))))))

(fn ShouldDie
  (params (Damage :type Float%20%28single-precision%29) (EventInstigator :type Controller%20Object%20Reference) (DamageCauser :type Actor%20Object%20Reference) (bIsTest :type Boolean) (EffectContext :type Gameplay%20Effect%20Context%20Handle%20Structure%20%28by%20ref%29))
  (return true))

(fn ConstructionScript)
