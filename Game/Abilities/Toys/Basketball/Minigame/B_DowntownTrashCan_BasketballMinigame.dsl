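; Downtown trash can Blueprint that acts as a hoop for the toy basketball minigame.
; When an actor enters the TriggerBox, the authority side casts it to B_ToyBasketball_Base.
; It then has the ball play its scoring cue and destruct, passing the trigger's location and a point offset below it along the can's up vector.
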(event OnComponentBeginOverlap(TriggerBox)
  (params (OverlappedComponent :type Primitive%20Component%20Object%20Reference) (OtherActor :type Actor%20Object%20Reference) (OtherComp :type Primitive%20Component%20Object%20Reference) (OtherBodyIndex :type Integer) (bFromSweep :type Boolean) (SweepResult :type Hit%20Result%20Structure%20%28by%20ref%29))
  (Networking|SwitchHasAuthority
    (:Authority
      (bind _asb_toy_basketball_base (Utilities|Casting|CastToB_ToyBasketball_Base OtherActor))
      (Class|BToyBasketballBase|PlayScoringCueandDestruct _asb_toy_basketball_base 0 (Transformation|GetWorldLocation (Variables|Default|GetTriggerBox)) (+ (Transformation|GetWorldLocation (Variables|Default|GetTriggerBox)) (* (Transformation|GetActorUpVector) -7e+01))))))

(fn ConstructionScript)
