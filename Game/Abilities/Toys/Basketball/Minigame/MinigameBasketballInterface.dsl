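; Blueprint interface for the basketball toy minigame, letting callers talk to minigame actors without knowing their class.
; It declares one message, PlayScoringCueandDestruct, passing hoop start and end position vectors.
; The name indicates implementers play a scoring cue and then destroy themselves; the interface itself holds no logic.
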(fn PlayScoringCueandDestruct
  (params (Hoop%20Start%20Position :type Vector) (Hoop%20End%20Position :type Vector)))
